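# Editor utility widget that batch-bakes animation sequences through a constrained Level Sequence into new RT_-prefixed assets.
# Run finds source and Control Rig target spawnable bindings by display name, then places each animation on the source, end to end.
# Playback range follows each clip while the target is exported and saved, using Python-built, transforms-only export options.
# A details view exposes inputs, persisted between sessions through a SaveGame slot.
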EditorUtilityWidgetBlueprint /AnimInEdTools/BakeAnimationSequences/EUW_BakeAnimationSequencesWithConstraints
graphs [
 Bake Animation {nodes [
   n1 = Bake Animation!() -> (
    then
    Binding:"Movie Scene Binding Proxy Structure (by ref)"
    PackagePath:String
    AssetName:String
    Evaluate All Skeletal Mesh:bool=~
    export_transforms:bool=~
    export_morph_targets:bool=~
    export_attribute_curves:bool=~
    export_material_curves:bool=~
    include_animation_names:Array of Strings
    exclude_animation_names:Array of Strings
   ) {is=title K2Node_FunctionEntry}
   > n2 = GetAssetTools()
   > n3 = CreateAsset(self=@n2, AssetName=@n1.AssetName, PackagePath=@n1.PackagePath, AssetClass=/Script/Engine.AnimSequence)
   > n4 = Cast To AnimSequence!(execute, Object=@n3) -> (then, CastFailed, AsAnim Sequence:Anim Sequence Object Reference)
   > GetEditorWorld(self=Get UnrealEditorSubsystem())
   > n5 = Execute Python Script!(
    execute
    PythonScript="import unreal\r\noptions = unreal.AnimSeqExportOption()\r\noptions.export_transforms = export_transforms\r\noptions.export_morph_targets = export_morph_targets\r\noptions.export_attribute_curves = export_attribute_curves\r\noptions.export_material_curves = export_material_curves\r\noptions.evaluate_all_skeletal_mesh_components = evaluate_all\r\nif include_animation_names:\r\n\toptions.include_animation_names = include_animation_names\r\nif exclude_animation_names:\r\n\toptions.exclude_animation_names = exclude_animation_names"
    evaluate_all:bool=@n1.Evaluate All Skeletal Mesh
    export_transforms:bool=@n1.export_transforms
    export_morph_targets:bool=@n1.export_morph_targets
    export_attribute_curves:bool=@n1.export_attribute_curves
    export_material_curves:bool=@n1.export_material_curves
    include_animation_names:Array of Strings=@n1.include_animation_names
    exclude_animation_names:Array of Strings=@n1.exclude_animation_names
   ) -> (then, ReturnValue, options:Anim Seq Export Option Object Reference) {
    inputs [
     evaluate_all
     export_transforms
     export_morph_targets
     export_attribute_curves
     export_material_curves
     include_animation_names
     exclude_animation_names
    ]
    outputs [options]
   }
   > ExportAnimSequenceFromSequencer!(
    execute
    AnimSequence:Anim Sequence Object Reference=@n4
    ExportOption:Anim Seq Export Option Object Reference=@n5.options
    Binding:"Movie Scene Binding Proxy Structure (by ref)"=@n1
    bCreateLink:bool=false
   ) -> (then, ReturnValue:bool)
   > Return Node!(execute, Anim Sequence:Anim Sequence Object Reference=@n4)
  ]}
 EventGraph {nodes [
   "On Released (DoIt_BTN)"() {is=title K2Node_ComponentBoundEvent}
   > OpenLevelSequence!(
    execute
    LevelSequence:Level Sequence Object Reference=Get Constrained Level Sequence() -> (Constrained Level Sequence:Level Sequence Object Reference)
   ) -> (then, ReturnValue:bool)
   > Set StartFrame(StartFrame:int=0) -> (Output_Get:int=0)
   > n1 = For Each Loop(Array:Array of Anim Sequence Object References=Get Animation Sequences() -> (Animation Sequences:Array of Anim Sequence Object References)) -> (Array Element:Anim Sequence Object Reference)
   > Set Anim Sequence(Anim Sequence:Anim Sequence Object Reference=@n1) -> (Output_Get:Anim Sequence Object Reference)
   > n2 = GetSpawnables!(execute, Sequence:Movie Scene Sequence Object Reference=Get Constrained Level Sequence()) -> (then, ReturnValue:Array of Movie Scene Binding Proxy Structures)
   > n3 = For Each Loop(Array:Array of Movie Scene Binding Proxy Structures=@n2) -> (Array Element:Movie Scene Binding Proxy Structure)
   n4 = AddTrack!(
    execute=@n3.Completed
    InBinding:"Movie Scene Binding Proxy Structure (by ref)"=Get SKM Binding() -> (SKM Binding:Movie Scene Binding Proxy Structure)
    TrackType:Movie Scene Track Class Reference=/Script/MovieSceneTracks.MovieSceneSkeletalAnimationTrack
   ) -> (then, ReturnValue:Movie Scene Track Object Reference)
   > n5 = AddSection!(execute, Track:Movie Scene Track Object Reference=@n4) -> (then, ReturnValue:Movie Scene Section Object Reference)
   > n6 = Cast To MovieSceneSkeletalAnimationSection!(execute, Object=@n5) -> (then, CastFailed, AsMovie Scene Skeletal Animation Section:Movie Scene Skeletal Animation Section Object Reference) {is=title K2Node_DynamicCast}
   > Set Section(Section:Movie Scene Skeletal Animation Section Object Reference=@n6) -> (Output_Get:Movie Scene Skeletal Animation Section Object Reference)
   n7 = Get Constrained Level Sequence()
   n8 = Get StartFrame() -> (StartFrame:int=0)
   > SetPlaybackStart!(execute, Sequence:Movie Scene Sequence Object Reference=@n7, StartFrame:int=@n8) -> (then)
   n9 = int + int(
    A=@n8
    B=GetNumFrames!(AnimationSequenceBase:Anim Sequence Base Object Reference=Get Anim Sequence() -> (Anim Sequence:Anim Sequence Object Reference)) -> (NumFrames:int=0)
   )
   > SetPlaybackEnd!(execute, Sequence:Movie Scene Sequence Object Reference=@n7, EndFrame:int=@n9) -> (then)
   n10 = int + int(A=@n9)
   > SetRange!(
    execute
    Section:Movie Scene Section Object Reference=Get Section() -> (Section:Movie Scene Skeletal Animation Section Object Reference)
    StartFrame:int=@n8
    EndFrame:int=@n10
   ) -> (then)
   > Set StartFrame(StartFrame=@n10) -> (Output_Get=0)
   n11 = Get Section()
   n12 = Get!(self:Movie Scene Skeletal Animation Section Object Reference=@n11) -> (
    Params_Animation:Anim Sequence Base Object Reference
    Params_FirstLoopStartFrameOffset:Frame Number Structure
    Params_StartFrameOffset:Frame Number Structure
    Params_EndFrameOffset:Frame Number Structure
    Params_PlayRate:Movie Scene Time Warp Variant Structure
    Params_bReverse:bool=~
    Params_SlotName:Name
    Params_MirrorDataTable:Mirror Data Table Object Reference
    Params_bSkipAnimNotifiers:bool=~
    Params_bForceCustomMode:bool=~
    Params_SwapRootBone:ESwapRootBone Enum
    Params_bLinearPlaybackWhenScaled:bool=~
   )
   > Set!(
    execute
    Params_Animation:Anim Sequence Base Object Reference=Get Anim Sequence()
    Params_FirstLoopStartFrameOffset:Frame Number Structure=@n12.Params_FirstLoopStartFrameOffset
    Params_StartFrameOffset:Frame Number Structure=@n12.Params_StartFrameOffset
    Params_EndFrameOffset:Frame Number Structure=@n12.Params_EndFrameOffset
    Params_PlayRate:Movie Scene Time Warp Variant Structure=@n12.Params_PlayRate
    Params_bReverse:bool=@n12.Params_bReverse
    Params_SlotName:Name=@n12.Params_SlotName
    Params_MirrorDataTable:Mirror Data Table Object Reference=@n12.Params_MirrorDataTable
    Params_bSkipAnimNotifiers:bool=@n12.Params_bSkipAnimNotifiers
    Params_bForceCustomMode:bool=@n12.Params_bForceCustomMode
    Params_SwapRootBone:ESwapRootBone Enum=@n12.Params_SwapRootBone
    Params_bLinearPlaybackWhenScaled:bool=False
    self:Movie Scene Skeletal Animation Section Object Reference=@n11
   ) -> (then, Output_Get:Movie Scene Skeletal Animation Params Structure)
   > RefreshCurrentLevelSequence!(execute) -> (then)
   n13 = Get Object Path String(Object=Get Anim Sequence())
   > PrintString(InString=@n13)
   > n14 = Bake Animation!(
    execute
    self
    Binding:"Movie Scene Binding Proxy Structure (by ref)"=Get MR Binding() -> (MR Binding:Movie Scene Binding Proxy Structure)
    PackagePath:String=Get Baked Animations Package Path() -> (Baked Animations Package Path:String)
    AssetName:String=Append(A=RT_, B=Split!(InPath:String=@n13) -> (PathPart:String, FilenamePart:String, ExtensionPart:String).FilenamePart)
    Evaluate All Skeletal Mesh:bool=true
    export_transforms:bool=true
    export_morph_targets:bool=false
    export_attribute_curves:bool=false
    export_material_curves:bool=false
    include_animation_names:Array of Strings
    exclude_animation_names:Array of Strings
   ) -> (then, Anim Sequence:Anim Sequence Object Reference) {is=title K2Node_CallFunction}
   > SaveLoadedAsset(AssetToSave=@n14, bOnlyIfIsDirty=true)
   n15 = GetName!(execute=@n3.LoopBody, InBinding:"Movie Scene Binding Proxy Structure (by ref)"=@n3) -> (then, ReturnValue:String)
   > n16 = Sequence()
   n17 = Get SKM Source() -> (SKM Source:Actor Object Reference)
   > Branch(Condition="Equal Exactly (String)"(A=@n15, B=GetDisplayName(Object=@n17)))
   > Set SKM Binding(SKM Binding:Movie Scene Binding Proxy Structure=@n3) -> (Output_Get:Movie Scene Binding Proxy Structure)
   n18 = Get Control Rig Target() -> (Control Rig Target:Actor Object Reference)
   Branch(execute=@n16.then_1, Condition="Equal Exactly (String)"(A=@n15, B=GetDisplayName(Object=@n18)))
   > Set MR Binding(MR Binding:Movie Scene Binding Proxy Structure=@n3) -> (Output_Get:Movie Scene Binding Proxy Structure)
   "On Property Changed (DetailsView_34)"!() -> (OutputDelegate, then, PropertyName:Name=None) {is=title K2Node_ComponentBoundEvent}
   n19 = Get Save Game() -> (Save Game:SG Bake Animation Sequences With Constraints Object Reference)
   > Set SKM Source!(execute, SKM Source:Actor Object Reference=Get SKM Source(), self:SG Bake Animation Sequences With Constraints Object Reference=@n19) -> (then, Output_Get:Actor Object Reference)
   > Set Control Rig Target!(
    execute
    Control Rig Target:Actor Object Reference=Get Control Rig Target()
    self:SG Bake Animation Sequences With Constraints Object Reference=@n19
   ) -> (then, Output_Get:Actor Object Reference)
   > Set Animation Sequences!(
    execute
    Animation Sequences:Array of Anim Sequence Object References=Get Animation Sequences()
    self:SG Bake Animation Sequences With Constraints Object Reference=@n19
   ) -> (then, Output_Get:Array of Anim Sequence Object References)
   > Set Baked Animations Package Path!(
    execute
    Baked Animations Package Path:String=Get Baked Animations Package Path()
    self:SG Bake Animation Sequences With Constraints Object Reference=@n19
   ) -> (then, Output_Get:String)
   > Set Constrained Animation Sequence!(
    execute
    Constrained Animation Sequence:Level Sequence Object Reference=Get Constrained Level Sequence()
    self:SG Bake Animation Sequences With Constraints Object Reference=@n19
   ) -> (then, Output_Get:Level Sequence Object Reference)
   > AsyncSaveGameToSlot!(
    execute:Exec
    WorldContextObject:Object Reference
    SaveGameObject:Save Game Object Reference=@n19
    SlotName:String=BakeAnimationSequences
    UserIndex:int=0
   ) -> (then:Exec, Completed:Exec, SaveGame:Save Game Object Reference, bSuccess:bool=~) {is=title K2Node_AsyncAction}
   Event PreConstruct()
   > SetObject(self=Get DetailsView_34() -> (DetailsView_34:Details View Object Reference), NewObject=Self-Reference())
   Event Construct()
   > n20 = LoadGameFromSlot!(execute, SlotName:String=BakeAnimationSequences, UserIndex:int=0) -> (then, ReturnValue:Save Game Object Reference)
   > n21 = Cast To SG_BakeAnimationSequencesWithConstraints!(execute, Object=@n20) -> (then, CastFailed, AsSG Bake Animation Sequences with Constraints:SG Bake Animation Sequences With Constraints Object Reference) {is=title K2Node_DynamicCast}
   > Set Save Game(Save Game:SG Bake Animation Sequences With Constraints Object Reference=@n21) -> (Output_Get:SG Bake Animation Sequences With Constraints Object Reference)
   > Set SKM Source!(execute, SKM Source=Get SKM Source!(self:SG Bake Animation Sequences With Constraints Object Reference=@n21) -> (SKM Source)) -> (then, Output_Get)
   > Set Control Rig Target!(
    execute
    Control Rig Target=Get Control Rig Target!(self:SG Bake Animation Sequences With Constraints Object Reference=@n21) -> (Control Rig Target)
   ) -> (then, Output_Get)
   > Set Baked Animations Package Path!(
    execute
    Baked Animations Package Path=Get Baked Animations Package Path!(self:SG Bake Animation Sequences With Constraints Object Reference=@n21) -> (Baked Animations Package Path)
   ) -> (then, Output_Get)
   > Set Animation Sequences!(
    execute
    Animation Sequences=Get Animation Sequences!(self:SG Bake Animation Sequences With Constraints Object Reference=@n21) -> (Animation Sequences)
   ) -> (then, Output_Get)
   > Set Constrained Level Sequence(Constrained Level Sequence:Level Sequence Object Reference=Get Constrained Animation Sequence!(self:SG Bake Animation Sequences With Constraints Object Reference=@n21) -> (Constrained Animation Sequence:Level Sequence Object Reference)) -> (Output_Get:Level Sequence Object Reference)
   n22 = CreateSaveGameObject!(
    execute=@n21.CastFailed
    SaveGameClass:Save Game Class Reference=&/SG_BakeAnimationSequencesWithConstraints.SG_BakeAnimationSequencesWithConstraints_C
   ) -> (then, ReturnValue:SG Bake Animation Sequences With Constraints Object Reference)
   > Set Save Game(Save Game=@n22)
   int - int(A=@n9)
  ]}
]
variables [
 Anim Sequence {type=AnimSequence, default=null}
 Animation Sequences {type="array<AnimSequence>", default []}
 Baked Animations Package Path {type=string, default=""}
 Constrained Level Sequence {type=LevelSequence, default=null}
 Control Rig Class {type=Class, default=null}
 Control Rig Target {type=Actor, default=null}
 Level Sequence {type=LevelSequence, default=null}
 MR Binding {type=MovieSceneBindingProxy, default {bindingId=#1, sequence=null}}
 Save Game {type=SG_BakeAnimationSequencesWithConstraints_C, default=null}
 Section {type=MovieSceneSkeletalAnimationSection, default=null, channels=~}
 SKM Binding {type=MovieSceneBindingProxy, default {bindingId=#1, sequence=null}}
 SKM Source {type=Actor, default=null}
 StartFrame {type=integer}
]
widgets {
 info {rootWidgetClass=&VerticalBox, widgetCount=4}
 widgets [
  VerticalBox_0 {widgetClassPath=&VerticalBox, widget=VerticalBox_0}
  DetailsView_34 {
   widgetClassPath=&DetailsView
   parent=VerticalBox_0
   widget=DetailsView_34
   properties {categoriesToShow [Inputs]}
   slot {class=VerticalBoxSlot}
  }
  DoIt_BTN {widgetClassPath=&EditorUtilityButton, parent=VerticalBox_0, widget=DoIt_BTN, slot {class=VerticalBoxSlot}}
  TextBlock_0 {widgetClassPath=&TextBlock, parent=DoIt_BTN, widget=TextBlock_0, properties {text=Run}, slot {class=ButtonSlot}}
 ]
}
